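# Niagara cosmetic glitter for the ReliableElk variant of the M_MED_Lettuce player body, with spawning gated by the User.ReliableElkVariant bool.
# Four sprite emitters (Foot_Shoes, Foot_Megaphone, Torso_Front, Torso_Back) share system spawn and size settings and pin short-lived spinning sparkles to skeletal mesh points using offset arrays.
# Foot_Shoes cycles bones by partition and mirrors offsets via spawn-group mask.
# Sparkles scale with camera distance, fade by curve and dim by time of day.
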NiagaraSystem /BRCosmetics/Characters/Player/Male/Medium/Bodies/M_MED_Lettuce/FX/NS_MED_ReliableElk_Glitter
dependencies {
 usedDataInterfaces [&NiagaraDataInterfaceCurve, &NiagaraDataInterfaceArrayInt32]
 usedDynamicInputs []
 usedModules [
  &self:SystemSpawnScript.SystemScriptSource.SystemScriptGraph.NiagaraNodeAssignment_0.SetVariables_#1
  &SystemState
  &self:SystemSpawnScript.SystemScriptSource.SystemScriptGraph.NiagaraNodeAssignment_1.SetVariables_#2
 ]
}
emitters {
 Foot_Megaphone {
  emitterUpdateScript [
   &EmitterState()
   &SpawnRate(
    Spawn Probability=$System.SpawnProbability
    SpawnRate=&Multiply_Float(
     A=$System.SpawnRate
     B=&MakeCustomFloatFromBool(Bool=$User.ReliableElkVariant, True Float=0.6, False Float=0)
    )
    Use Spawn Probability=true
   )
  ]
  particleSpawnScript [
   &PartitionParticles(Particles Per Partition=1, Partition Mode:ENiagara_PartitionMode=::NewEnumerator1)
   &InitializeParticle(
    Color Mode=::NewEnumerator0
    Lifetime=&RandomRangeFloat(Minimum=0.3, Maximum=0.4)
    Sprite Rotation Mode=::NewEnumerator0
    Sprite Size Mode=::NewEnumerator1
    Uniform Sprite Size Max=$System.MaxSize
    Uniform Sprite Size Min=$System.MinSize
    UsePositionOffset=false
    Write Material Random=true
   )
   &CameraOffset(Camera Offset Amount=$System.CameraOffset)
  ]
  particleUpdateScript [
   &ParticleState()
   &SkeletalMeshLocation(
    Bone / Socket Index=5
    Bone Sampling Mode=::NewEnumerator4
    MaskSpawnGroup=false
    Sampled Position Offset=&SelectVectorFromArray(
     Array Sampling Mode=::NewEnumerator0
     Direct Array Index=?
     Vector Selection Array=DataInterface() {propertyValues {FloatData [<10.6607 -8.1098 5.5512>, <-0.4384 -0.7379 6.6005>, <10.0345 1.2042 9.263>, <6.8034 1.2042 3.2778>, <9.1977 -3.2274 0.9969>]}}
     Random Seed=0
    )
    Skeletal Mesh=$System.Skeletal Mesh 1
   ) {module=SkeletalMeshLocation004}
   &SpriteRotationRate(Rotation Rate=250)
   &ScaleSpriteSize(
    Scale Sprite Size Mode=::NewEnumerator0
    Uniform Scale Factor=&FloatFromCurve(
     FloatCurve=DataInterface() {propertyValues {Curve {keys [{arriveTangent=2.9879, interpMode=RCIM_Cubic, leaveTangent=2.9879, tangentMode=RCTM_User, time=0, value=0}, <at 1 1 RCIM_Cubic>]}}}
     Scale Curve=&FN_LOD_Distance_Remap(Start Distance=100, End Distance=500, Minimum Output=0.75, Maximum Output=1.25, Clamp Output=true)
    )
   )
   &ScaleColor(
    Scale Alpha=&FloatFromCurve(FloatCurve=DataInterface() {propertyValues {Curve {keys [
         {arriveTangent=9.0725, interpMode=RCIM_Cubic, leaveTangent=9.0725, tangentMode=RCTM_User, time=0, value=0}
         {arriveTangent=-2.5249, interpMode=RCIM_Cubic, leaveTangent=-2.5249, tangentMode=RCTM_User, time=0.1408, value=0.9847}
         {arriveTangent=-0.0405, interpMode=RCIM_Cubic, leaveTangent=-0.0405, tangentMode=RCTM_User, time=1, value=0}
        ]}}})
    Scale RGB=&Multiply_VectorByFloat(
     Vector=<1 1 1>
     Float=&FNTimeOfDayScalar(Night Scale=0.2, Morning Scale=0.5, Day Scale=0.5, Evening Scale=0.5)
    )
    ScaleA=true
    ScaleRGB=true
   )
  ]
  rendererClasses [&NiagaraSpriteRendererProperties]
  renderers [{rendererClass=&NiagaraSpriteRendererProperties}]
 }
 Foot_Shoes {
  emitterUpdateScript [
   &EmitterState()
   &SpawnRate(
    Spawn Probability=$System.SpawnProbability
    SpawnRate=&Multiply_Float(
     A=$System.SpawnRate
     B=&MakeCustomFloatFromBool(Bool=$User.ReliableElkVariant, True Float=1, False Float=0)
    )
    Use Spawn Probability=true
   )
  ]
  particleSpawnScript [
   &PartitionParticles(Particles Per Partition=1, Partition Mode:ENiagara_PartitionMode=::NewEnumerator1)
   &InitializeParticle(
    Color Mode=::NewEnumerator0
    Lifetime=&RandomRangeFloat(Minimum=0.3, Maximum=0.4)
    Sprite Rotation Mode=::NewEnumerator0
    Sprite Size Mode=::NewEnumerator1
    Uniform Sprite Size Max=$System.MaxSize
    Uniform Sprite Size Min=$System.MinSize
    UsePositionOffset=false
    Write Material Random=true
   )
   &CameraOffset(Camera Offset Amount=$System.CameraOffset)
  ]
  particleUpdateScript [
   &ParticleState()
   &SkeletalMeshLocation(
    Bone / Socket Index=&SelectIntFromArray(Direct Array Index=$Particles.PartitionParticles.Partition, Int Selection Array=DataInterface() {propertyValues {IntData [1,2,3,4]}})
    Bone Sampling Mode=::NewEnumerator4
    MaskSpawnGroup=false
    Sampled Position Offset=&Multiply_Vector(
     A=&SelectVectorFromArray(Array Sampling Mode=::NewEnumerator0, Direct Array Index=?, Vector Selection Array=$User.Array_Foot, Random Seed=0)
     B=&MaskVectorBySpawnGroup(Masked Vector=<1 1 1>, Passthrough Vector=<-1 -1 -1>, Spawn Group Mask=$Particles.PartitionParticles.Partition)
    )
    Skeletal Mesh=$System.Skeletal Mesh 1
   ) {module=SkeletalMeshLocation004}
   &SpriteRotationRate(Rotation Rate=250)
   &ScaleSpriteSize(
    Scale Sprite Size Mode=::NewEnumerator0
    Uniform Scale Factor=&FloatFromCurve(
     FloatCurve=DataInterface() {propertyValues {Curve {keys [{arriveTangent=2.9879, interpMode=RCIM_Cubic, leaveTangent=2.9879, tangentMode=RCTM_User, time=0, value=0}, <at 1 1 RCIM_Cubic>]}}}
     Scale Curve=&FN_LOD_Distance_Remap(Start Distance=100, End Distance=500, Minimum Output=0.75, Maximum Output=1.25, Clamp Output=true)
    )
   )
   &ScaleColor(
    Scale Alpha=&FloatFromCurve(FloatCurve=DataInterface() {propertyValues {Curve {keys [
         {arriveTangent=9.0725, interpMode=RCIM_Cubic, leaveTangent=9.0725, tangentMode=RCTM_User, time=0, value=0}
         {arriveTangent=-2.5249, interpMode=RCIM_Cubic, leaveTangent=-2.5249, tangentMode=RCTM_User, time=0.1408, value=0.9847}
         {arriveTangent=-0.0405, interpMode=RCIM_Cubic, leaveTangent=-0.0405, tangentMode=RCTM_User, time=1, value=0}
        ]}}})
    Scale RGB=&Multiply_VectorByFloat(
     Vector=<1 1 1>
     Float=&FNTimeOfDayScalar(Night Scale=0.2, Morning Scale=0.5, Day Scale=0.5, Evening Scale=0.5)
    )
    ScaleA=true
    ScaleRGB=true
   )
  ]
  rendererClasses [&NiagaraSpriteRendererProperties]
  renderers [{rendererClass=&NiagaraSpriteRendererProperties}]
 }
 Torso_Back {
  emitterUpdateScript [
   &EmitterState()
   &SpawnRate(
    Spawn Probability=$System.SpawnProbability
    SpawnRate=&Multiply_Float(
     A=$System.SpawnRate
     B=&MakeCustomFloatFromBool(Bool=$User.ReliableElkVariant, True Float=0.4, False Float=0)
    )
    Use Spawn Probability=true
   )
  ]
  particleSpawnScript [
   &PartitionParticles(Number Of Partitions=12, Particles Per Partition=1, Partition Mode:ENiagara_PartitionMode=::NewEnumerator1)
   &InitializeParticle(
    Color Mode=::NewEnumerator0
    Lifetime=&RandomRangeFloat(Minimum=0.3, Maximum=0.4)
    Position Mode=::NewEnumerator0
    Sprite Rotation Mode=::NewEnumerator0
    Sprite Size Mode=::NewEnumerator1
    Uniform Sprite Size Max=$System.MaxSize
    Uniform Sprite Size Min=$System.MinSize
    Write Material Random=true
   )
  ]
  particleUpdateScript [
   &ParticleState()
   &SkeletalMeshLocation(
    Bone Sampling Mode=::NewEnumerator4
    MaskSpawnGroup=false
    Sampled Position Offset=&SelectVectorFromArray(
     Array Sampling Mode=::NewEnumerator0
     Direct Array Index=$Particles.Initial.PartitionParticles.Partition
     Vector Selection Array=DataInterface() {propertyValues {FloatData [<8.8191 -12 -7.38>, <7.8333 -12 5.4349>, <4.3846 -11.5 0.4031>, <1.2877 -11.5 -3.0688>]}}
     Random Seed=0
    )
    Skeletal Mesh=$System.Skeletal Mesh 1
   ) {module=SkeletalMeshLocation004}
   &SpriteRotationRate(Rotation Rate=250)
   &ScaleSpriteSize(
    Scale Sprite Size Mode=::NewEnumerator0
    Uniform Scale Factor=&FloatFromCurve(
     FloatCurve=DataInterface() {propertyValues {Curve {keys [{arriveTangent=2.9879, interpMode=RCIM_Cubic, leaveTangent=2.9879, tangentMode=RCTM_User, time=0, value=0}, <at 1 1 RCIM_Cubic>]}}}
     Scale Curve=&FN_LOD_Distance_Remap(Start Distance=100, End Distance=500, Minimum Output=0.75, Maximum Output=1.25, Clamp Output=true)
    )
   )
   &ScaleColor(
    Scale Alpha=&FloatFromCurve(FloatCurve=DataInterface() {propertyValues {Curve {keys [
         {arriveTangent=9.0725, interpMode=RCIM_Cubic, leaveTangent=9.0725, tangentMode=RCTM_User, time=0, value=0}
         {arriveTangent=-2.5249, interpMode=RCIM_Cubic, leaveTangent=-2.5249, tangentMode=RCTM_User, time=0.1408, value=0.9847}
         {arriveTangent=-0.0405, interpMode=RCIM_Cubic, leaveTangent=-0.0405, tangentMode=RCTM_User, time=1, value=0}
        ]}}})
    Scale RGB=&Multiply_VectorByFloat(
     Vector=<1 1 1>
     Float=&FNTimeOfDayScalar(Night Scale=0.2, Morning Scale=0.5, Day Scale=0.5, Evening Scale=0.5)
    )
    ScaleA=true
    ScaleRGB=true
   )
  ]
  rendererClasses [&NiagaraSpriteRendererProperties]
  renderers [{rendererClass=&NiagaraSpriteRendererProperties}]
 }
 Torso_Front {
  emitterUpdateScript [
   &EmitterState()
   &SpawnRate(
    Spawn Probability=$System.SpawnProbability
    SpawnRate=&Multiply_Float(
     A=$System.SpawnRate
     B=&MakeCustomFloatFromBool(Bool=$User.ReliableElkVariant, True Float=0.4, False Float=0)
    )
    Use Spawn Probability=true
   )
  ]
  particleSpawnScript [
   &PartitionParticles(Number Of Partitions=12, Particles Per Partition=1, Partition Mode:ENiagara_PartitionMode=::NewEnumerator1)
   &InitializeParticle(
    Color Mode=::NewEnumerator0
    Lifetime=&RandomRangeFloat(Minimum=0.3, Maximum=0.4)
    Position Mode=::NewEnumerator0
    Sprite Rotation Mode=::NewEnumerator0
    Sprite Size Mode=::NewEnumerator1
    Uniform Sprite Size Max=$System.MaxSize
    Uniform Sprite Size Min=$System.MinSize
    Write Material Random=true
   )
  ]
  particleUpdateScript [
   &ParticleState()
   &SkeletalMeshLocation(
    Bone Sampling Mode=::NewEnumerator4
    MaskSpawnGroup=false
    Sampled Position Offset=&SelectVectorFromArray(
     Array Sampling Mode=::NewEnumerator0
     Direct Array Index=$Particles.Initial.PartitionParticles.Partition
     Vector Selection Array=DataInterface() {propertyValues {FloatData [
        <6.5315 11.6734 8.9822>
        <3.2226 12.32 6.8016>
        <3.3384 12.7759 2.6243>
        <6.8666 12.5551 5.0307>
        <7.1046 12.5 -0.9307>
        <2.5994 12.858 -2.9119>
        <4.2807 12.5647 -6.1513>
        <7.2441 12.5372 -4.057>
        <6.6553 10.9643 -10.0279>
        <2.8992 11.5518 -9.001>
       ]}}
     Random Seed=0
    )
    Skeletal Mesh=$System.Skeletal Mesh 1
   ) {module=SkeletalMeshLocation004}
   &SpriteRotationRate(Rotation Rate=250)
   &ScaleSpriteSize(
    Scale Sprite Size Mode=::NewEnumerator0
    Uniform Scale Factor=&FloatFromCurve(
     FloatCurve=DataInterface() {propertyValues {Curve {keys [{arriveTangent=2.9879, interpMode=RCIM_Cubic, leaveTangent=2.9879, tangentMode=RCTM_User, time=0, value=0}, <at 1 1 RCIM_Cubic>]}}}
     Scale Curve=&FN_LOD_Distance_Remap(Start Distance=100, End Distance=500, Minimum Output=0.75, Maximum Output=1.25, Clamp Output=true)
    )
   )
   &ScaleColor(
    Scale Alpha=&FloatFromCurve(FloatCurve=DataInterface() {propertyValues {Curve {keys [
         {arriveTangent=9.0725, interpMode=RCIM_Cubic, leaveTangent=9.0725, tangentMode=RCTM_User, time=0, value=0}
         {arriveTangent=-2.5249, interpMode=RCIM_Cubic, leaveTangent=-2.5249, tangentMode=RCTM_User, time=0.1408, value=0.9847}
         {arriveTangent=-0.0405, interpMode=RCIM_Cubic, leaveTangent=-0.0405, tangentMode=RCTM_User, time=1, value=0}
        ]}}})
    Scale RGB=&Multiply_VectorByFloat(
     Vector=<1 1 1>
     Float=&FNTimeOfDayScalar(Night Scale=0.2, Morning Scale=0.5, Day Scale=0.5, Evening Scale=0.5)
    )
    ScaleA=true
    ScaleRGB=true
   )
  ]
  rendererClasses [&NiagaraSpriteRendererProperties]
  renderers [{rendererClass=&NiagaraSpriteRendererProperties}]
 }
}
summary {
 systemName=NS_MED_ReliableElk_Glitter
 emitters [
  {emitterName=Torso_Back, rendererClasses [&NiagaraSpriteRendererProperties]}
  {emitterName=Foot_Shoes, rendererClasses [&NiagaraSpriteRendererProperties]}
  {emitterName=Torso_Front, rendererClasses [&NiagaraSpriteRendererProperties]}
  {emitterName=Foot_Megaphone, rendererClasses [&NiagaraSpriteRendererProperties]}
 ]
 userVariables [
  User.Array_Foot {
   type=NiagaraDataInterfaceArrayFloat3
   description=""
   defaultValue {
    struct=NiagaraExt_VariableValue_DataInterface
    value {
     dataInterface=&self:NiagaraDataInterfaceArrayFloat3_0
     dataInterfaceClass=&NiagaraDataInterfaceArrayFloat3
    }
   }
  }
  User.ReliableElkVariant {type=NiagaraBool, defaultValue=true}
  User.Skeletal Mesh 1 {
   type=NiagaraDataInterfaceSkeletalMesh
   defaultValue {
    struct=NiagaraExt_VariableValue_DataInterface
    value {dataInterface=&self:NiagaraDataInterfaceSkeletalMesh_0}
   }
  }
 ]
}
