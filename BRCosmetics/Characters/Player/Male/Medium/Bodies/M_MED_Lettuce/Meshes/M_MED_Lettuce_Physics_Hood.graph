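# Physics asset for the M_MED_Lettuce male medium cosmetic body, setting up physics for its hood bones.
# Capsules sit on the central, left and right hood bones, with further capsules on the upper spine, neck and both clavicles.
# Each hood bone is constrained directly to spine_05 with limited swing and very tight twist.
# Other listed bodies, including limbs, pelvis, lower spine and a megaphone bone, carry no shapes.
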PhysicsAsset /BRCosmetics/Characters/Player/Male/Medium/Bodies/M_MED_Lettuce/Meshes/M_MED_Lettuce_Physics_Hood
bodies {
 calf_l []
 calf_r []
 clavicle_l [{shapeType=Capsule, length=7.5, radius=8.5, shapeName=clavicle_l_capsule, center=<6.5034 -0.0001 0>, rotation=<rot -90 166.7374 -166.7379>}]
 clavicle_r [{shapeType=Capsule, length=7.5, radius=8.5, shapeName=clavicle_l_capsule, center=<-6.5034 0.0001 0>, rotation=<rot -90 6.7215 173.2779>}]
 dyn_hood_01 [{shapeType=Capsule, length=5, radius=3.5, shapeName=dyn_hood_01_capsule, center=<-3.0399 -2.0809 0>, rotation=<rot -90 26.5651 13.4337>}]
 dyn_hood_l_01 [{
   shapeType=Capsule
   length=3.75
   radius=3.25
   shapeName=dyn_hood_l_01_capsule
   center=<-2.0115 5.1401 -2.3431>
   rotation=<rot 3.8415 -14.5127 104.5103>
  }]
 dyn_hood_r_01 [{
   shapeType=Capsule
   length=3.75
   radius=3.25
   shapeName=dyn_hood_l_01_capsule
   center=<2.0115 -5.1401 2.3431>
   rotation=<rot 3.8415 -14.5127 -75.4897>
  }]
 dyn_megaphone []
 foot_l []
 foot_r []
 lowerarm_l []
 lowerarm_r []
 neck_01 [{shapeType=Capsule, length=8.75, radius=4, shapeName=neck_01_capsule, center=<4.8934 0 0>, rotation=<rot -90 0.0004 -0.0008>}]
 pelvis []
 spine_01 []
 spine_02 []
 spine_03 []
 spine_04 []
 spine_05 [{shapeType=Capsule, length=18.9707, radius=9.5, shapeName=spine_05_capsule}]
 thigh_l []
 thigh_r []
 upperarm_l []
 upperarm_r []
}
constraints [
 {bone1Name=dyn_hood_l_01, bone2Name=spine_05, swing1LimitDegrees=20, swing2LimitDegrees=15, twistLimitDegrees=2}
 {bone1Name=dyn_hood_01, bone2Name=spine_05, swing1LimitDegrees=20, swing2LimitDegrees=15, twistLimitDegrees=2}
 {bone1Name=dyn_hood_r_01, bone2Name=spine_05, swing1LimitDegrees=20, swing2LimitDegrees=15, twistLimitDegrees=2}
]
