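# Physics asset for the megaphone on the M_MED_Lettuce medium male player cosmetic body, covering only that prop and the right thigh.
# Capsule bodies sit on the dyn_megaphone bone and on the right thigh and thigh twist bones.
# A single constraint ties the megaphone to thigh_r with wide swing and narrow twist limits, which suggests dangling secondary motion against the leg.
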PhysicsAsset /BRCosmetics/Characters/Player/Male/Medium/Bodies/M_MED_Lettuce/Meshes/M_MED_Lettuce_Physics_Megaphone
bodies {
 dyn_megaphone [{shapeType=Capsule, length=9, radius=8.25, shapeName=dyn_megaphone_capsule, center=<8 0 7>, rotation=<rot 90 0 20>}]
 thigh_r [{
   shapeType=Capsule
   length=23.25
   radius=7.25
   shapeName=thigh_twist_01_r_capsule
   center=<21.6982 -0.0002 0>
   rotation=<rot 90 179.9207 179.9213>
  }]
 thigh_twist_01_r [{shapeType=Capsule, length=23.25, radius=7.25, shapeName=thigh_twist_01_r_capsule, rotation=<rot 90 88.3838 88.3844>}]
}
constraints [{bone1Name=dyn_megaphone, bone2Name=thigh_r, swing1LimitDegrees=60, swing2LimitDegrees=60, twistLimitDegrees=10}]
