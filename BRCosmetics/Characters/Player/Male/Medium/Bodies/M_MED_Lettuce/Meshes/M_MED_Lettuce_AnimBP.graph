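# Post-processes the copied body pose for the Lettuce male cosmetic, animating hood, megaphone, shoe backs and thigh deform bones.
# Chains constraints, additive bone offsets and two RigidBody simulations before a deform Control Rig.
# Each update derives physics and constraint alphas from speed and melee state, hood easing from vertical velocity, and correctives from bone deltas.
# Customization loading caches meshes; the hood-back corrective applies only with the Lettuce hat equipped.
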AnimBlueprint /BRCosmetics/Characters/Player/Male/Medium/Bodies/M_MED_Lettuce/Meshes/M_MED_Lettuce_AnimBP
target_skeleton=&/M_MED_Lettuce_Skeleton
graphs [
 AnimGraph {nodes [
   n1 = "Constraint - gravity (TT)"!(
    ComponentPose=Local To Component(LocalPose=Copy Pose From Mesh() {bUseAttachedParent=true})
    Alpha=1
    ConstraintWeights_0=0.8
    ConstraintWeights_1:float=0.2
   ) -> (Pose) {
    alphaBoolBlend {blendOption=Linear}
    boneToModify {boneName=gravity}
    constraintSetup [{transformType=Translation, targetBone {boneName=thigh_r}}, {transformType=Translation, targetBone {boneName=root}}]
    constraintWeights [1,1]
   }
   n2 = "Constraint - dyn_shoe_back_l (PP)"!(ComponentPose=@n1, Alpha=1, ConstraintWeights_0=0.5, ConstraintWeights_1:float=0.5) -> (Pose) {
    alphaBoolBlend {blendOption=Linear}
    boneToModify {boneName=dyn_shoe_back_l}
    constraintSetup [{targetBone {boneName=calf_twist_01_l}}, {targetBone {boneName=dyn_shoe_back_l}}]
    constraintWeights [1,1]
   }
   n3 = "Constraint - dyn_shoe_back_r (PP)"!(ComponentPose=@n2, Alpha=1, ConstraintWeights_0=0.5, ConstraintWeights_1:float=0.5) -> (Pose) {
    alphaBoolBlend {blendOption=Linear}
    boneToModify {boneName=dyn_shoe_back_r}
    constraintSetup [{targetBone {boneName=calf_twist_01_r}}, {targetBone {boneName=dyn_shoe_back_r}}]
    constraintWeights [1,1]
   }
   n4 = "Constraint - deform_Left_base_l (PP)"!(ComponentPose=@n3, Alpha=1, ConstraintWeights_0=0.5, ConstraintWeights_1:float=0.5) -> (Pose) {
    alphaBoolBlend {blendOption=Linear}
    boneToModify {boneName=deform_Left_base_l}
    constraintSetup [{targetBone {boneName=thigh_twist_01_l}}, {targetBone {boneName=pelvis}}]
    constraintWeights [1,1]
   }
   n5 = "Constraint - deform_Right_base_r (PP)"!(ComponentPose=@n4, Alpha=1, ConstraintWeights_0=0.5, ConstraintWeights_1:float=0.5) -> (Pose) {
    alphaBoolBlend {blendOption=Linear}
    boneToModify {boneName=deform_Right_base_r}
    constraintSetup [{targetBone {boneName=thigh_twist_01_r}}, {targetBone {boneName=pelvis}}]
    constraintWeights [1,1]
   }
   n6 = "Transform (Modify) Bone - Bone: deform_Left_l"(ComponentPose=@n5, Alpha=Get thigh_shape_lt_l() -> (thigh_shape_lt_l:double), Translation="0,0,-5", Rotation="0,0,0", Scale="1,1,1") {
    rotationMode=BMM_Additive
    rotationSpace=BCS_BoneSpace
    translationMode=BMM_Additive
    translationSpace=BCS_BoneSpace
    alphaBoolBlend {blendOption=Linear}
    boneToModify {boneName=deform_Left_l}
    translation {z=-5}
   }
   n7 = "Transform (Modify) Bone - Bone: deform_Right_r"(ComponentPose=@n6, Alpha=Get thigh_shape_rt_r() -> (thigh_shape_rt_r:double), Translation="0,0,5", Rotation="0,0,0", Scale="1,1,1") {
    rotationMode=BMM_Additive
    rotationSpace=BCS_BoneSpace
    translationMode=BMM_Additive
    translationSpace=BCS_BoneSpace
    alphaBoolBlend {blendOption=Linear}
    boneToModify {boneName=deform_Right_r}
    translation {z=5}
   }
   n8 = "Transform (Modify) Bone - Bone: dyn_hood_02"(ComponentPose=@n7, Alpha=Get head_shape_bk() -> (head_shape_bk:double), Translation="0,5,0", Rotation="0,0,0", Scale="1,1,1") {
    rotationMode=BMM_Additive
    rotationSpace=BCS_BoneSpace
    translationMode=BMM_Additive
    translationSpace=BCS_BoneSpace
    alphaBoolBlend {blendOption=Linear}
    boneToModify {boneName=dyn_hood_02}
    translation {y=5}
   }
   n9 = RigidBody(ComponentPose=@n8, Alpha=Get rigid_body_alpha() -> (rigid_body_alpha:double)) {
    bClampLinearTranslationLimitToRefPose=true
    bOverrideWorldGravity=true
    overridePhysicsAsset=&/M_MED_Lettuce_Physics_Hood
    alphaBoolBlend {blendOption=Linear}
    alphaScaleBiasClamp {bInterpResult=true}
    componentAppliedLinearAccClamp=<3000 3000 3000>
    componentLinearAccScale=<0.3 0.3 0.3>
    componentLinearVelScale=<0.3 0.3 0.3>
    overrideWorldGravity {z=-980}
   }
   n10 = RigidBody(ComponentPose=@n9) {
    bClampLinearTranslationLimitToRefPose=true
    bOverrideWorldGravity=true
    overridePhysicsAsset=&/M_MED_Lettuce_Physics_Megaphone
    simulationSpace=BaseBoneSpace
    alphaBoolBlend {blendOption=Linear}
    alphaScaleBiasClamp {bInterpResult=true}
    baseBoneRef {boneName=thigh_r}
    componentAppliedLinearAccClamp=<3000 3000 3000>
    componentLinearAccScale=<0.3 0.3 0.3>
    componentLinearVelScale=<0.3 0.3 0.3>
    overrideWorldGravity {z=-980}
    simSpaceSettings {velocityScaleZ=0.5, worldAlpha=1}
   }
   n11 = Get Constraint alpha() -> (Constraint alpha:double)
   n12 = "Constraint - dyn_megaphone (P)"(ComponentPose=@n10, Alpha=@n11, ConstraintWeights_0=1) {
    alphaBoolBlend {blendOption=Linear}
    boneToModify {boneName=dyn_megaphone}
    constraintSetup [{targetBone {boneName=pelvis}}]
    constraintWeights [1]
   }
   n13 = "Transform (Modify) Bone - Bone: dyn_megaphone"(ComponentPose=@n12, Alpha=@n11, Translation="0,0,0", Rotation="10,0,0", Scale="1,1,1") {
    rotationMode=BMM_Additive
    rotationSpace=BCS_BoneSpace
    translationMode=BMM_Additive
    translationSpace=BCS_BoneSpace
    alphaBoolBlend {blendOption=Linear}
    boneToModify {boneName=dyn_megaphone}
    rotation {pitch=10}
   }
   n14 = Get is in jump or falling() -> (is in jump or falling:bool)
   n15 = Get jump easing tr() -> (jump easing tr:double)
   n16 = "Transform (Modify) Bone - Bone: dyn_hood_01"!(
    ComponentPose=@n13
    bAlphaBoolEnabled:bool=@n14
    Translation_X:double=0
    Translation_Y:double=0
    Translation_Z:double=@n15
    Rotation_Roll:float=0
    Rotation_Pitch:float=0
    Rotation_Yaw:float=Get jump easing() -> (jump easing:double)
    Scale="1,1,1"
   ) -> (Pose) {
    alphaInputType=Bool
    rotationMode=BMM_Additive
    rotationSpace=BCS_BoneSpace
    translationMode=BMM_Additive
    translationSpace=BCS_WorldSpace
    alphaBoolBlend {blendInTime=0.4, blendOption=Linear, blendOutTime=0.4}
    boneToModify {boneName=dyn_hood_01}
   }
   n17 = Get side jump easing() -> (side jump easing:double)
   n18 = "Transform (Modify) Bone - Bone: dyn_hood_l_01"!(
    ComponentPose=@n16
    bAlphaBoolEnabled:bool=@n14
    Translation_X:double=0
    Translation_Y:double=0
    Translation_Z:double=@n15
    Rotation_Roll:float=0
    Rotation_Pitch:float=0
    Rotation_Yaw:float=@n17
    Scale="1,1,1"
   ) -> (Pose) {
    alphaInputType=Bool
    rotationMode=BMM_Additive
    rotationSpace=BCS_BoneSpace
    translationMode=BMM_Additive
    translationSpace=BCS_WorldSpace
    alphaBoolBlend {blendInTime=0.4, blendOption=Linear, blendOutTime=0.4}
    boneToModify {boneName=dyn_hood_l_01}
   }
   n19 = "Transform (Modify) Bone - Bone: dyn_hood_r_01"!(
    ComponentPose=@n18
    bAlphaBoolEnabled:bool=@n14
    Translation_X:double=0
    Translation_Y:double=0
    Translation_Z:double=@n15
    Rotation_Roll:float=0
    Rotation_Pitch:float=0
    Rotation_Yaw:float=@n17
    Scale="1,1,1"
   ) -> (Pose) {
    alphaInputType=Bool
    rotationMode=BMM_Additive
    rotationSpace=BCS_BoneSpace
    translationSpace=BCS_WorldSpace
    alphaBoolBlend {blendInTime=0.4, blendOption=Linear, blendOutTime=0.4}
    boneToModify {boneName=dyn_hood_r_01}
   }
   n20 = "Transform (Modify) Bone - Bone: dyn_shoe_back_l"!(
    ComponentPose=@n19
    bAlphaBoolEnabled:bool=Get is left foot bends back() -> (is left foot bends back:bool)
    Translation="0,0,0"
    Rotation_Roll:float=0
    Rotation_Pitch:float=0
    Rotation_Yaw:float=Get shoe back rot l() -> (shoe back rot l:double)
    Scale="1,1,1"
   ) -> (Pose) {
    alphaInputType=Bool
    rotationMode=BMM_Additive
    rotationSpace=BCS_BoneSpace
    alphaBoolBlend {blendOption=Linear}
    boneToModify {boneName=dyn_shoe_back_l}
    rotation {yaw=70}
   }
   n21 = "Transform (Modify) Bone - Bone: dyn_shoe_back_r"!(
    ComponentPose=@n20
    bAlphaBoolEnabled:bool=Get is right foot bends back() -> (is right foot bends back:bool)
    Translation="0,0,0"
    Rotation_Roll:float=0
    Rotation_Pitch:float=0
    Rotation_Yaw:float=Get shoe back rot r() -> (shoe back rot r:double)
    Scale="1,1,1"
   ) -> (Pose) {
    alphaInputType=Bool
    rotationMode=BMM_Additive
    rotationSpace=BCS_BoneSpace
    alphaBoolBlend {blendOption=Linear}
    boneToModify {boneName=dyn_shoe_back_r}
    rotation {yaw=70}
   }
   n22 = "Transform (Modify) Bone - Bone: dyn_hood_01"!(ComponentPose=@n21, Alpha:float=1, Translation:Vector="-5,-2,0", Rotation:Rotator="0,0,0", Scale="1,1,1") -> (Pose) {
    rotationMode=BMM_Additive
    rotationSpace=BCS_BoneSpace
    translationMode=BMM_Additive
    translationSpace=BCS_BoneSpace
    alphaBoolBlend {blendOption=Linear}
    boneToModify {boneName=dyn_hood_01}
    translation=<-5 -2>
   }
   n23 = "Transform (Modify) Bone - Bone: dyn_hood_l_01"!(ComponentPose=@n22, Alpha:float=1, Translation:Vector="3,1,0", Rotation:Rotator="0,0,0", Scale="1,1,1") -> (Pose) {
    rotationMode=BMM_Additive
    rotationSpace=BCS_BoneSpace
    translationMode=BMM_Additive
    translationSpace=BCS_BoneSpace
    alphaBoolBlend {blendOption=Linear}
    boneToModify {boneName=dyn_hood_l_01}
    translation=<3 1>
   }
   n24 = "Transform (Modify) Bone - Bone: dyn_hood_r_01"!(ComponentPose=@n23, Alpha:float=1, Translation:Vector="-3,-1,0", Rotation:Rotator="0,0,0", Scale="1,1,1") -> (Pose) {
    rotationMode=BMM_Additive
    rotationSpace=BCS_BoneSpace
    translationMode=BMM_Additive
    translationSpace=BCS_BoneSpace
    alphaBoolBlend {blendOption=Linear}
    boneToModify {boneName=dyn_hood_r_01}
    translation=<-3 -1>
   }
   n25 = Control Rig(Source=Component To Local(ComponentPose=@n24)) {
    alphaBoolBlend {blendOption=Linear}
    controlRigAssetReference {blueprintRigClass=&/M_MED_DeformRig_Master_ControlRig_Lettuce.M_MED_DeformRig_Master_ControlRig_Lettuce_C}
   }
   Output Pose(Result=@n25) {layerGroup=None}
  ]}
 EventGraph {nodes [
   Event BlueprintUpdateAnimation()
   n1 = Get bIsPlayingMeleeAnim()
   n2 = OR Boolean(A=Get bIsJumping(), B=Get bIsFalling())
   n3 = NOT Boolean(A=@n2)
   n4 = Get bIsMoving2D()
   n5 = float * float(A=Vector Length(A=Get PawnVelocity()), B="To Float (Boolean)"(InBool=AND Boolean(A=@n4, B=@n3))) {is=title K2Node_PromotableOperator}
   > Set rigid_body_alpha(rigid_body_alpha:double=float * float!(A=SelectFloat(A=0.2, B=0.7, bPickA=@n1), B=MapRangeClamped(Value=@n5, InRangeA=550, InRangeB=770, OutRangeA=0.7, OutRangeB=0.3), C:double) -> (ReturnValue)) -> (Output_Get:double)
   > Set Constraint alpha(Constraint alpha:double=MapRangeClamped(Value=@n5, InRangeA=30, InRangeB=550, OutRangeB=0.3)) -> (Output_Get:double)
   > Set jump easing(jump easing:double=MapRangeClamped(Value=Get PawnVelocityZ(), InRangeA=630, InRangeB=-630, OutRangeA=15, OutRangeB=-15)) -> (Output_Get:double)
   > Set side jump easing(side jump easing:double=MapRangeClamped(Value=Get PawnVelocityZ(), InRangeA=630, InRangeB=-630, OutRangeA=8, OutRangeB=-8)) -> (Output_Get:double)
   > Set jump easing tr(jump easing tr:double=MapRangeClamped(Value=Get PawnVelocityZ(), InRangeA=630, InRangeB=-630, OutRangeA=2, OutRangeB=0)) -> (Output_Get:double)
   > Set is in jump or falling(is in jump or falling:bool=@n2) -> (Output_Get:bool)
   n6 = GetOwnerSkeletalMeshComponent()
   n7 = BreakRotator(InRot=BreakTransform(InTransform=GetDeltaTransformFromRefPose(self=@n6, BoneName=foot_r, BaseName=calf_r)).Rotation)
   > Set is right foot bends back(is right foot bends back:bool="float > float"(A=@n7.Yaw)) -> (Output_Get:bool)
   > Set shoe back rot r(shoe back rot r:double=float * float!(A=@n7.Yaw, B=-1.5) -> (ReturnValue)) -> (Output_Get:double)
   n8 = BreakRotator(InRot=BreakTransform(InTransform=GetDeltaTransformFromRefPose(self=@n6, BoneName=foot_l, BaseName=calf_l)).Rotation)
   > Set is left foot bends back(is left foot bends back:bool="float > float"(A=@n8.Yaw)) -> (Output_Get:bool)
   > Set shoe back rot l(shoe back rot l:double=float * float(A=@n8.Yaw, B=-1.5)) -> (Output_Get:double)
   n9 = Get Character Mesh() -> (Character Mesh:Skeletal Mesh Component Object Reference)
   > Set thigh_shape_lt_l(thigh_shape_lt_l:double=GetAlphaFromReferenceTransformDelta(
     SkeletalMeshComponent=@n9
     BoneName=thigh_l
     BaseName=pelvis
     DeltaComponentType=EulerY
     InputRange=(X=60,Y=0)
     OutputRange=(X=1,Y=0)
     TwistAxis="0,0,1"
    )) -> (Output_Get:double)
   > n10 = Set thigh_shape_rt_r(thigh_shape_rt_r:double=GetAlphaFromReferenceTransformDelta(
     SkeletalMeshComponent=@n9
     BoneName=thigh_r
     BaseName=pelvis
     DeltaComponentType=EulerY
     InputRange=(X=60,Y=0)
     OutputRange=(X=1,Y=0)
     TwistAxis="0,0,1"
    )) -> (Output_Get:double)
   n11 = Event OnCharacterCustomizationLoaded()
   n12 = Branch(
    execute=@n10.then
    Condition="Equal (Object)"(
     A=Get Array Item with EFort Custom Part Type(Array:Array of Object References=@n11.CustomParts, PartType=Hat) -> (Output:Object Reference)
     B=/BRCosmetics/Characters/CharacterParts/FaceAccessories/CP_M_MED_Lettuce_Hat_FaceAcc
    )
   )
   > Set head_shape_bk(head_shape_bk:double=GetAlphaFromReferenceTransformDelta(
     SkeletalMeshComponent=Get Head Mesh() -> (Head Mesh:Skeletal Mesh Component Object Reference)
     BoneName=head
     BaseName=spine_05
     InputRange=(X=0,Y=-80)
     TwistAxis="0,0,1"
    )) -> (Output_Get:double)
   Set head_shape_bk(execute=@n12.else, head_shape_bk=0)
   GetSkeletalMesh!(self:Custom Character Part Object Reference=Get Array Item with EFort Custom Part Type(Array:Array of Custom Character Part Object References=@n11.CustomParts, PartType=Head) -> (Output:Custom Character Part Object Reference)) -> (ReturnValue:Skeletal Mesh Object Reference)
   Set Character Mesh(execute=@n13.then_1, Character Mesh:Skeletal Mesh Component Object Reference=@n11.CustomBaseMesh) -> (Output_Get:Skeletal Mesh Component Object Reference)
   Set Head Mesh(
    execute=@n13.then_2
    Head Mesh:Skeletal Mesh Component Object Reference=Get Array Item with EFort Custom Part Type(Array:Array of Skeletal Mesh Component Budgeted Object References=@n11.CustomMeshParts, PartType=Head) -> (Output:Skeletal Mesh Component Budgeted Object Reference)
   ) -> (Output_Get:Skeletal Mesh Component Object Reference)
   n13 = Sequence!(execute=@n11.then) -> (then_0, then_1, then_2:Exec, then_3:Exec)
  ]}
]
variables [
 Character Mesh {type=SkeletalMeshComponent}
 Constraint alpha {type=number}
 Head Mesh {type=SkeletalMeshComponent}
 head_shape_bk {type=number}
 is in jump or falling {type=boolean}
 is left foot bends back {type=boolean}
 is right foot bends back {type=boolean}
 jump easing {type=number}
 jump easing tr {type=number}
 rigid_body_alpha {type=number}
 shoe back rot l {type=number}
 shoe back rot r {type=number}
 side jump easing {type=number}
 thigh_shape_lt_l {type=number}
 thigh_shape_rt_r {type=number}
]
