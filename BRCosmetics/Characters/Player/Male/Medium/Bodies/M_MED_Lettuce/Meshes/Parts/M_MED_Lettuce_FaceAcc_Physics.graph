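# Physics asset for the face accessory part of the M_MED_Lettuce medium male player cosmetic, covering its dynamic hair strand bones.
# Capsule bodies sit on the head, the hair base and two two-bone strands, fr and l.
# Constraints with narrow swing and tighter twist limits chain each strand to the hair base.
# The fr strand's root is also constrained to the head, unlike the l strand.
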PhysicsAsset /BRCosmetics/Characters/Player/Male/Medium/Bodies/M_MED_Lettuce/Meshes/Parts/M_MED_Lettuce_FaceAcc_Physics
bodies {
 dyn_hair_base [{shapeType=Capsule, length=6.5627, radius=3.4671, shapeName=dyn_hair_base_capsule, rotation=<rot 0.0001 0 -88.0397>}]
 dyn_strand_01_fr [{shapeType=Capsule, length=1.1, radius=1, shapeName=dyn_strand_01_fr_capsule, rotation=<rot 90 0 0>}]
 dyn_strand_01_l [{shapeType=Capsule, length=1.1, radius=1, shapeName=dyn_strand_01_l_capsule, rotation=<rot 90 0 0>}]
 dyn_strand_02_fr [{shapeType=Capsule, length=1.3886, radius=0.7114, shapeName=dyn_strand_02_fr_capsule, rotation=<rot 90 0 0>}]
 dyn_strand_02_l [{shapeType=Capsule, length=1.3898, radius=0.7102, shapeName=dyn_strand_02_l_capsule, rotation=<rot 90 0 0>}]
 head [{
   shapeType=Capsule
   length=8.5771
   radius=5.8962
   shapeName=head_capsule
   center=<4.0345 4.9765 0.7831>
   rotation=<rot -84.7523 179.99 168.2391>
  }]
}
constraints [
 {bone1Name=dyn_strand_02_fr, bone2Name=dyn_strand_01_fr, swing1LimitDegrees=15, swing2LimitDegrees=15, twistLimitDegrees=5}
 {bone1Name=dyn_strand_01_fr, bone2Name=head, swing1LimitDegrees=15, swing2LimitDegrees=15, twistLimitDegrees=5}
 {bone1Name=dyn_strand_01_fr, bone2Name=dyn_hair_base, swing1LimitDegrees=15, swing2LimitDegrees=15, twistLimitDegrees=5}
 {bone1Name=dyn_strand_01_l, bone2Name=dyn_hair_base, swing1LimitDegrees=15, swing2LimitDegrees=15, twistLimitDegrees=5}
 {bone1Name=dyn_strand_02_l, bone2Name=dyn_strand_01_l, swing1LimitDegrees=15, swing2LimitDegrees=15, twistLimitDegrees=5}
]
