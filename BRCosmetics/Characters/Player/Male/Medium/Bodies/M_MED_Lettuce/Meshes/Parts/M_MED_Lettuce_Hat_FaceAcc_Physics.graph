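# Physics asset for the hat and face accessory part of the medium male Lettuce character cosmetic, defining bodies for secondary motion of hat, rope and hair.
# An eight-link rope chain hangs from Hat with tight swing on the first two joints and looser swing below, while front and left two-link hair strands branch from dyn_hair_base.
# Hat also links to three hat bones.
# Unconstrained head, neck, clavicle and upper-spine capsules supply body collision shapes.
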PhysicsAsset /BRCosmetics/Characters/Player/Male/Medium/Bodies/M_MED_Lettuce/Meshes/Parts/M_MED_Lettuce_Hat_FaceAcc_Physics
bodies {
 Hat [{shapeType=Capsule, length=8.0192, radius=1, shapeName=Hat_capsule, rotation=<rot 0 0 90>}]
 clavicle_l [{shapeType=Capsule, length=10, radius=7, shapeName=clavicle_l_capsule, center=<12.7729 0 0>, rotation=<rot 0 90 90>}]
 clavicle_r [{shapeType=Capsule, length=10, radius=7, shapeName=clavicle_r_capsule, center=<-10 -1 0>, rotation=<rot 90 0 0>}]
 dyn_hair_base [{shapeType=Capsule, length=0.9738, radius=0.673, shapeName=dyn_hair_base_capsule, rotation=<rot 90 0 0>}]
 dyn_hat_03 [{shapeType=Capsule, length=22, radius=3.4, shapeName=dyn_hat_03_capsule, center=<0 3 0>, rotation=<rot 12 0.0002 -0.0001>}]
 dyn_hat_04 [{shapeType=Capsule, length=12, radius=3, shapeName=dyn_hat_04_capsule, center=<0 3 0>, rotation=<rot -8 0 0>}]
 dyn_rope_01 [{shapeType=Capsule, length=7.3083, radius=1.9105, shapeName=dyn_rope_01_capsule}]
 dyn_rope_02 [{shapeType=Capsule, length=7.3083, radius=1.9105, shapeName=dyn_rope_02_capsule}]
 dyn_rope_03 [{shapeType=Capsule, length=7.6177, radius=1.6011, shapeName=dyn_rope_03_capsule, center=<0 0.2102 0>}]
 dyn_rope_04 [{shapeType=Capsule, length=7.6842, radius=1.5347, shapeName=dyn_rope_04_capsule, center=<0 0.2102 0>}]
 dyn_rope_05 [{shapeType=Capsule, length=7.8245, radius=1.3943, shapeName=dyn_rope_05_capsule, center=<0 0.2102 0>}]
 dyn_rope_06 [{shapeType=Capsule, length=1.2553, radius=1.4902, shapeName=dyn_rope_06_capsule, rotation=<rot 90 0 0>}]
 dyn_rope_07 [{shapeType=Capsule, length=1.0878, radius=1.6578, shapeName=dyn_rope_07_capsule, center=<1.9358 0.0241 0>, rotation=<rot 90 0 0>}]
 dyn_rope_08 [{shapeType=Capsule, length=1.0878, radius=1.6578, shapeName=dyn_rope_08_capsule, center=<1.9358 0.0241 0>, rotation=<rot 90 0 0>}]
 dyn_strand_01_fr [{shapeType=Capsule, length=0.9738, radius=0.673, shapeName=dyn_strand_01_fr_capsule, rotation=<rot 90 0 0>}]
 dyn_strand_01_l [{shapeType=Capsule, length=0.9738, radius=0.673, shapeName=dyn_strand_01_l_capsule, rotation=<rot 90 0 0>}]
 dyn_strand_02_fr [{shapeType=Capsule, length=0.9738, radius=0.673, shapeName=dyn_strand_02_fr_capsule, rotation=<rot 90 0 0>}]
 dyn_strand_02_l [{shapeType=Capsule, length=0.9738, radius=0.673, shapeName=dyn_strand_02_l_capsule, rotation=<rot 90 0 0>}]
 hat_02 [{shapeType=Capsule, length=9.5689, radius=4, shapeName=hat_02_capsule, center=<0 1 0>, rotation=<rot 0 -90 0>}]
 head [{
   shapeType=Capsule
   length=7.2162
   radius=6.3708
   shapeName=head_capsule
   center=<5.3946 4.0718 -0.0588>
   rotation=<rot 89.3752 -0.0004 9.0016>
  }]
 neck_01 [
  {shapeType=Capsule, length=18.3693, radius=5.6984, shapeName=neck_01_capsule, center=<6.4205 0 0.1002>, rotation=<rot -86.896 179.9999 180>}
  {shapeType=Capsule, length=14.1918, radius=5.2405, shapeName=neck_01_capsule, center=<-5.9134 -7.9993 0>}
 ]
 spine_04 [{shapeType=Capsule, length=13.3706, radius=11.7187, shapeName=spine_04_capsule, center=<6.5925 0 0>}]
 spine_05 [
  {shapeType=Capsule, length=10.9332, radius=9.1438, shapeName=spine_05_capsule, center=<5.9712 0.3874 0>}
  {
   shapeType=Capsule
   length=5.5539
   radius=2.8765
   shapeName=spine_05_capsule
   center=<7.855 7.3802 3.3307>
   rotation=<rot -36.8926 -0.0008 -16.495>
  }
  {
   shapeType=Capsule
   length=5.7662
   radius=2.7012
   shapeName=spine_05_capsule
   center=<7.7893 7.1914 -5.9081>
   rotation=<rot 31.3592 -0.0018 23.2584>
  }
  {shapeType=Capsule, length=6.4065, radius=6.624, shapeName=spine_05_capsule, center=<11.9809 0.8304 1.4324>}
 ]
}
constraints [
 {bone1Name=dyn_strand_01_l, bone2Name=dyn_hair_base, swing1LimitDegrees=15, swing2LimitDegrees=15, twistLimitDegrees=5}
 {bone1Name=dyn_strand_01_fr, bone2Name=dyn_hair_base, swing1LimitDegrees=15, swing2LimitDegrees=15, twistLimitDegrees=5}
 {bone1Name=dyn_strand_02_fr, bone2Name=dyn_strand_01_fr, swing1LimitDegrees=15, swing2LimitDegrees=15, twistLimitDegrees=5}
 {bone1Name=dyn_strand_02_l, bone2Name=dyn_strand_01_l, swing1LimitDegrees=15, swing2LimitDegrees=15, twistLimitDegrees=5}
 {bone1Name=dyn_rope_01, bone2Name=Hat, swing1LimitDegrees=20, swing2LimitDegrees=0, twistLimitDegrees=5}
 {bone1Name=dyn_rope_02, bone2Name=dyn_rope_01, swing1LimitDegrees=1, swing2LimitDegrees=10, twistLimitDegrees=8}
 {bone1Name=dyn_rope_03, bone2Name=dyn_rope_02, swing1LimitDegrees=60, swing2LimitDegrees=60, twistLimitDegrees=10}
 {bone1Name=dyn_rope_04, bone2Name=dyn_rope_03, swing1LimitDegrees=60, swing2LimitDegrees=60, twistLimitDegrees=5}
 {bone1Name=dyn_rope_05, bone2Name=dyn_rope_04, swing1LimitDegrees=60, swing2LimitDegrees=60, twistLimitDegrees=5}
 {bone1Name=dyn_rope_06, bone2Name=dyn_rope_05, swing1LimitDegrees=60, swing2LimitDegrees=60, twistLimitDegrees=5}
 {bone1Name=dyn_rope_07, bone2Name=dyn_rope_06, swing1LimitDegrees=60, swing2LimitDegrees=60, twistLimitDegrees=5}
 {bone1Name=dyn_rope_08, bone2Name=dyn_rope_07, swing1LimitDegrees=60, swing2LimitDegrees=60, twistLimitDegrees=5}
 {bone1Name=Hat, bone2Name=dyn_hat_03, swing1LimitDegrees=60, swing2LimitDegrees=60, twistLimitDegrees=60}
 {bone1Name=Hat, bone2Name=dyn_hat_04, swing1LimitDegrees=60, swing2LimitDegrees=30, twistLimitDegrees=10}
 {bone1Name=Hat, bone2Name=hat_02, swing1LimitDegrees=30, swing2LimitDegrees=120, twistLimitDegrees=20}
]
