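# Post-process style AnimBP for the Lettuce outfit's hat accessory, copying the attached parent's pose and nudging hat bones to follow body poses.
# bRigidBodyEnabled picks between rigid body simulation with corrective bone offsets, and a fallback using more offsets, a hat constraint and a rope trail controller.
# Each update derives corrective alphas from head and clavicle deltas, speed, sprint, targeting, emote and frontend state.
# Some bone chains and head-velocity logic are left unconnected.
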AnimBlueprint /BRCosmetics/Characters/Player/Male/Medium/Bodies/M_MED_Lettuce/Meshes/Parts/M_MED_Lettuce_Hat_FaceAcc_AnimBP
target_skeleton=&/M_MED_Lettuce_Hat_FaceAcc_Skeleton
graphs [
 AnimGraph {nodes [
   n1 = "Transform (Modify) Bone - Bone: hat_02"(Alpha=1, Translation="3,0,0", Rotation="0,0,0", Scale="1,1,1") {
    rotationSpace=BCS_BoneSpace
    translationMode=BMM_Additive
    translationSpace=BCS_ParentBoneSpace
    alphaBoolBlend {blendInTime=0.2, blendOption=Linear, blendOutTime=0.2}
    alphaScaleBiasClamp {bInterpResult=true, interpSpeedDecreasing=5, interpSpeedIncreasing=30}
    boneToModify {boneName=hat_02}
    translation {x=3}
   }
   n2 = "Transform (Modify) Bone - Bone: dyn_hat_03"(ComponentPose=@n1, Alpha=1, Translation="2,0,0", Rotation="0,0,0", Scale="1,1,1") {
    rotationSpace=BCS_BoneSpace
    translationMode=BMM_Additive
    translationSpace=BCS_ParentBoneSpace
    alphaBoolBlend {blendInTime=0.2, blendOption=Linear, blendOutTime=0.2}
    alphaScaleBiasClamp {bInterpResult=true, interpSpeedDecreasing=5, interpSpeedIncreasing=30}
    boneToModify {boneName=dyn_hat_03}
    translation {x=2}
   }
   "Transform (Modify) Bone - Bone: dyn_hat_04"(ComponentPose=@n2, Alpha=1, Translation="2,0,0", Rotation="0,0,0", Scale="1,1,1") {
    rotationSpace=BCS_BoneSpace
    translationMode=BMM_Additive
    translationSpace=BCS_ParentBoneSpace
    alphaBoolBlend {blendInTime=0.2, blendOption=Linear, blendOutTime=0.2}
    alphaScaleBiasClamp {bInterpResult=true, interpSpeedDecreasing=5, interpSpeedIncreasing=30}
    boneToModify {boneName=dyn_hat_04}
    translation {x=2}
   }
   n3 = Get head_shape_bk() -> (head_shape_bk:double)
   n4 = "Transform (Modify) Bone - Bone: hat_02"(Alpha=@n3, Translation="10,0,0", Rotation="0,0,0", Scale="1,1,1") {
    rotationMode=BMM_Additive
    rotationSpace=BCS_ParentBoneSpace
    translationMode=BMM_Additive
    translationSpace=BCS_ParentBoneSpace
    alphaBoolBlend {blendInTime=0.2, blendOption=Linear, blendOutTime=0.2}
    alphaScaleBiasClamp {bInterpResult=true, interpSpeedDecreasing=5, interpSpeedIncreasing=30}
    boneToModify {boneName=hat_02}
    translation {x=10}
   }
   n5 = "Transform (Modify) Bone - Bone: dyn_hat_03"(ComponentPose=@n4, Alpha=@n3, Translation="10,0,0", Rotation="0,0,0", Scale="1,1,1") {
    rotationMode=BMM_Additive
    rotationSpace=BCS_ParentBoneSpace
    translationMode=BMM_Additive
    translationSpace=BCS_ParentBoneSpace
    alphaBoolBlend {blendInTime=0.2, blendOption=Linear, blendOutTime=0.2}
    alphaScaleBiasClamp {bInterpResult=true, interpSpeedDecreasing=5, interpSpeedIncreasing=30}
    boneToModify {boneName=dyn_hat_03}
    translation {x=10}
   }
   "Transform (Modify) Bone - Bone: dyn_hat_04"(ComponentPose=@n5, Alpha=@n3, Translation="5,0,0", Rotation="0,0,0", Scale="1,1,1") {
    rotationMode=BMM_Additive
    rotationSpace=BCS_ParentBoneSpace
    translationMode=BMM_Additive
    translationSpace=BCS_ParentBoneSpace
    alphaBoolBlend {blendInTime=0.2, blendOption=Linear, blendOutTime=0.2}
    alphaScaleBiasClamp {bInterpResult=true, interpSpeedDecreasing=5, interpSpeedIncreasing=30}
    boneToModify {boneName=dyn_hat_04}
    translation {x=5}
   }
   n6 = RigidBody!(
    ComponentPose=Local To Component(LocalPose=Copy Pose From Mesh() {bUseAttachedParent=true})
    Alpha=Get RigidBodyAlpha() -> (RigidBodyAlpha:double)
    OverrideWorldGravity:Vector="0,0,-980"
    ExternalForce
   ) -> (Pose) {
    bClampLinearTranslationLimitToRefPose=true
    bOverrideWorldGravity=true
    overridePhysicsAsset=&/M_MED_Lettuce_Hat_FaceAcc_Physics
    simulationSpace=BaseBoneSpace
    alphaBoolBlend {blendOption=Linear}
    alphaScaleBiasClamp {bInterpResult=true, interpSpeedIncreasing=2}
    baseBoneRef {boneName=neck_02}
    componentAppliedLinearAccClamp=<1500 1500 1500>
    componentLinearAccScale=<0.3 0.3 0.3>
    componentLinearVelScale=<0.3 0.3 0.3>
    overrideWorldGravity {z=-980}
   }
   n7 = "Transform (Modify) Bone - Bone: hat_02"(ComponentPose=@n6, Alpha=Get Hat_transform() -> (Hat_transform:double), Translation="3.2,0,0", Rotation="0,0,0", Scale="1,1,1") {
    rotationMode=BMM_Additive
    rotationSpace=BCS_ParentBoneSpace
    translationMode=BMM_Additive
    translationSpace=BCS_ParentBoneSpace
    alphaBoolBlend {blendOption=Linear}
    boneToModify {boneName=hat_02}
    translation {x=3.2}
   }
   n8 = "Transform (Modify) Bone - Bone: hat_02"(
    ComponentPose=@n7
    Alpha=Get Crouch_Hat's_transform() -> (Crouch_Hat's_transform:double)
    Translation="4,0,0"
    Rotation="20,0,0"
    Scale="1,1,1"
   ) {
    rotationMode=BMM_Additive
    rotationSpace=BCS_ParentBoneSpace
    translationMode=BMM_Additive
    translationSpace=BCS_ParentBoneSpace
    alphaBoolBlend {blendOption=Linear}
    boneToModify {boneName=hat_02}
    rotation {pitch=20}
    translation {x=4}
   }
   n9 = "Transform (Modify) Bone - Bone: Hat"(ComponentPose=@n8, Alpha=Get TargetingCorrection() -> (TargetingCorrection:float), Translation="1,0,0", Rotation="0,0,0", Scale="1,1,1") {
    rotationMode=BMM_Additive
    rotationSpace=BCS_BoneSpace
    translationMode=BMM_Additive
    translationSpace=BCS_BoneSpace
    alphaBoolBlend {blendInTime=0.2, blendOption=Linear, blendOutTime=0.2}
    alphaScaleBiasClamp {bInterpResult=true, interpSpeedIncreasing=30}
    boneToModify {boneName=Hat}
    translation {x=1}
   }
   n10 = "Transform (Modify) Bone - Bone: hat_02"(
    ComponentPose=@n9
    Alpha=Get Hat_transform_Frontend() -> (Hat_transform_Frontend:double)
    Translation="1,0,0"
    Rotation="0,0,0"
    Scale="1,1,1"
   ) {
    rotationMode=BMM_Additive
    rotationSpace=BCS_ParentBoneSpace
    translationMode=BMM_Additive
    translationSpace=BCS_ParentBoneSpace
    alphaBoolBlend {blendOption=Linear}
    alphaScaleBiasClamp {bInterpResult=true}
    boneToModify {boneName=hat_02}
    translation {x=1}
   }
   n11 = Get Emote_Hat_transform() -> (Emote_Hat_transform:double)
   n12 = "Transform (Modify) Bone - Bone: Hat"(ComponentPose=@n10, Alpha=@n11, Translation="1,0,0", Rotation="0,0,0", Scale="1,1,1") {
    rotationMode=BMM_Additive
    rotationSpace=BCS_BoneSpace
    translationMode=BMM_Additive
    translationSpace=BCS_BoneSpace
    alphaBoolBlend {blendInTime=0.2, blendOption=Linear, blendOutTime=0.2}
    alphaScaleBiasClamp {bInterpResult=true, interpSpeedDecreasing=9, interpSpeedIncreasing=9}
    boneToModify {boneName=Hat}
    translation {x=1}
   }
   n13 = "Transform (Modify) Bone - Bone: hat_02"(ComponentPose=@n12, Alpha=@n11, Translation="2,0,0", Rotation="0,0,0", Scale="1,1,1") {
    rotationMode=BMM_Additive
    rotationSpace=BCS_ParentBoneSpace
    translationMode=BMM_Additive
    translationSpace=BCS_ParentBoneSpace
    alphaBoolBlend {blendInTime=0.2, blendOption=Linear, blendOutTime=0.2}
    alphaScaleBiasClamp {bInterpResult=true, interpSpeedDecreasing=9, interpSpeedIncreasing=9}
    boneToModify {boneName=hat_02}
    translation {x=2}
   }
   n14 = "Transform (Modify) Bone - Bone: dyn_hat_03"(
    ComponentPose=@n13
    Alpha=Get Emote_Hat_transform_2() -> (Emote_Hat_transform_2:double)
    Translation="3,0,0"
    Rotation="0,0,0"
    Scale="1,1,1"
   ) {
    rotationMode=BMM_Additive
    rotationSpace=BCS_ParentBoneSpace
    translationMode=BMM_Additive
    translationSpace=BCS_ParentBoneSpace
    alphaBoolBlend {blendInTime=0.2, blendOption=Linear, blendOutTime=0.2}
    alphaScaleBiasClamp {bInterpResult=true, interpSpeedDecreasing=9, interpSpeedIncreasing=9}
    boneToModify {boneName=dyn_hat_03}
    translation {x=3}
   }
   n15 = "Transform (Modify) Bone - Bone: hat_02"!(ComponentPose=@n14, bAlphaBoolEnabled:bool=True, Translation="5,0,0", Rotation="0,0,0", Scale="1,1,1") -> (Pose) {
    alphaInputType=Bool
    rotationMode=BMM_Additive
    rotationSpace=BCS_ParentBoneSpace
    translationSpace=BCS_ParentBoneSpace
    alphaBoolBlend {blendOption=CircularIn}
    alphaScaleBiasClamp {bInterpResult=true}
    boneToModify {boneName=hat_02}
    translation {x=5}
   }
   n16 = "Transform (Modify) Bone - Bone: Hat"!(ComponentPose=@n15, bAlphaBoolEnabled:bool=True, Translation="1,0,0", Rotation="0,2,0", Scale="1,1,1") -> (Pose) {
    alphaInputType=Bool
    rotationMode=BMM_Additive
    rotationSpace=BCS_ParentBoneSpace
    translationMode=BMM_Additive
    translationSpace=BCS_ParentBoneSpace
    alphaBoolBlend {blendOption=CircularIn}
    alphaScaleBiasClamp {bInterpResult=true}
    boneToModify {boneName=Hat}
    rotation {yaw=2}
    translation {x=1}
   }
   n17 = Get head_shape_bk()
   n18 = "Transform (Modify) Bone - Bone: hat_02"!(
    ComponentPose=Local To Component(LocalPose=Copy Pose From Mesh() {bUseAttachedParent=true})
    Alpha:float=@n17
    Translation="10,0,0"
    Rotation="0,0,0"
    Scale="1,1,1"
   ) -> (Pose) {
    rotationMode=BMM_Additive
    rotationSpace=BCS_ParentBoneSpace
    translationMode=BMM_Additive
    translationSpace=BCS_ParentBoneSpace
    alphaBoolBlend {blendInTime=0.2, blendOption=Linear, blendOutTime=0.2}
    alphaScaleBiasClamp {bInterpResult=true, interpSpeedDecreasing=5, interpSpeedIncreasing=30}
    boneToModify {boneName=hat_02}
    translation {x=10}
   }
   n19 = "Transform (Modify) Bone - Bone: dyn_hat_03"(ComponentPose=@n18, Alpha=@n17, Translation="10,0,0", Rotation="0,0,0", Scale="1,1,1") {
    rotationMode=BMM_Additive
    rotationSpace=BCS_ParentBoneSpace
    translationMode=BMM_Additive
    translationSpace=BCS_ParentBoneSpace
    alphaBoolBlend {blendInTime=0.2, blendOption=Linear, blendOutTime=0.2}
    alphaScaleBiasClamp {bInterpResult=true, interpSpeedDecreasing=5, interpSpeedIncreasing=30}
    boneToModify {boneName=dyn_hat_03}
    translation {x=10}
   }
   n20 = "Transform (Modify) Bone - Bone: dyn_hat_04"(ComponentPose=@n19, Alpha=@n17, Translation="5,0,0", Rotation="0,0,0", Scale="1,1,1") {
    rotationMode=BMM_Additive
    rotationSpace=BCS_ParentBoneSpace
    translationMode=BMM_Additive
    translationSpace=BCS_ParentBoneSpace
    alphaBoolBlend {blendInTime=0.2, blendOption=Linear, blendOutTime=0.2}
    alphaScaleBiasClamp {bInterpResult=true, interpSpeedDecreasing=5, interpSpeedIncreasing=30}
    boneToModify {boneName=dyn_hat_04}
    translation {x=5}
   }
   n21 = Get Hat_transform_noRB_emote() -> (Hat_transform_noRB_emote:double)
   n22 = "Transform (Modify) Bone - Bone: hat_02"(ComponentPose=@n20, Alpha=@n21, Translation="5,0,0", Rotation="0,0,-10", Scale="1,1,1") {
    rotationMode=BMM_Additive
    rotationSpace=BCS_ParentBoneSpace
    translationMode=BMM_Additive
    translationSpace=BCS_ParentBoneSpace
    alphaBoolBlend {blendOption=Linear}
    alphaScaleBiasClamp {bInterpResult=true, interpSpeedDecreasing=2}
    boneToModify {boneName=hat_02}
    rotation {roll=-10}
    translation {x=5}
   }
   n23 = "Transform (Modify) Bone - Bone: dyn_hat_03"(ComponentPose=@n22, Alpha=@n21, Translation="5,0,0", Rotation="0,0,0", Scale="1,1,1") {
    rotationMode=BMM_Additive
    rotationSpace=BCS_ParentBoneSpace
    translationMode=BMM_Additive
    translationSpace=BCS_ParentBoneSpace
    alphaBoolBlend {blendOption=Linear}
    alphaScaleBiasClamp {bInterpResult=true, interpSpeedDecreasing=2}
    boneToModify {boneName=dyn_hat_03}
    translation {x=5}
   }
   n24 = "Transform (Modify) Bone - Bone: dyn_hat_04"(ComponentPose=@n23, Alpha=@n21, Translation="5,0,0", Rotation="0,0,0", Scale="1,1,1") {
    rotationMode=BMM_Additive
    rotationSpace=BCS_ParentBoneSpace
    translationMode=BMM_Additive
    translationSpace=BCS_ParentBoneSpace
    alphaBoolBlend {blendOption=Linear}
    alphaScaleBiasClamp {bInterpResult=true, interpSpeedDecreasing=2}
    boneToModify {boneName=dyn_hat_04}
    translation {x=5}
   }
   n25 = Get Hat_transform_Level() -> (Hat_transform_Level:double)
   n26 = "Transform (Modify) Bone - Bone: dyn_hat_03"(ComponentPose=@n24, Alpha=@n25, Translation="5,0,0", Rotation="0,0,0", Scale="1,1,1") {
    rotationMode=BMM_Additive
    rotationSpace=BCS_ParentBoneSpace
    translationMode=BMM_Additive
    translationSpace=BCS_ParentBoneSpace
    alphaBoolBlend {blendOption=Linear}
    alphaScaleBiasClamp {interpSpeedIncreasing=2}
    boneToModify {boneName=dyn_hat_03}
    translation {x=5}
   }
   n27 = "Transform (Modify) Bone - Bone: dyn_hat_04"(ComponentPose=@n26, Alpha=@n25, Translation="5,0,0", Rotation="0,0,0", Scale="1,1,1") {
    rotationMode=BMM_Additive
    rotationSpace=BCS_ParentBoneSpace
    translationMode=BMM_Additive
    translationSpace=BCS_ParentBoneSpace
    alphaBoolBlend {blendOption=Linear}
    alphaScaleBiasClamp {interpSpeedIncreasing=2}
    boneToModify {boneName=dyn_hat_04}
    translation {x=5}
   }
   n28 = "Transform (Modify) Bone - Bone: hat_02"(ComponentPose=@n27, Alpha=@n25, Translation="5,0,0", Rotation="0,0,0", Scale="1,1,1") {
    rotationMode=BMM_Additive
    rotationSpace=BCS_ParentBoneSpace
    translationMode=BMM_Additive
    translationSpace=BCS_ParentBoneSpace
    alphaBoolBlend {blendOption=Linear}
    alphaScaleBiasClamp {interpSpeedIncreasing=2}
    boneToModify {boneName=hat_02}
    translation {x=5}
   }
   n29 = Get Hat_transform_Level_headUp() -> (Hat_transform_Level_headUp:double)
   n30 = "Transform (Modify) Bone - Bone: hat_02"(ComponentPose=@n28, Alpha=@n29, Translation="6.7,0,0", Rotation="0,0,0", Scale="1,1,1") {
    rotationMode=BMM_Additive
    rotationSpace=BCS_ParentBoneSpace
    translationMode=BMM_Additive
    translationSpace=BCS_ParentBoneSpace
    alphaBoolBlend {blendOption=Linear}
    alphaScaleBiasClamp {bInterpResult=true, interpSpeedDecreasing=2}
    boneToModify {boneName=hat_02}
    translation {x=6.7}
   }
   n31 = "Transform (Modify) Bone - Bone: dyn_hat_04"(ComponentPose=@n30, Alpha=@n29, Translation="2,0,0", Rotation="0,0,0", Scale="1,1,1") {
    rotationMode=BMM_Additive
    rotationSpace=BCS_ParentBoneSpace
    translationMode=BMM_Additive
    translationSpace=BCS_ParentBoneSpace
    alphaBoolBlend {blendOption=Linear}
    alphaScaleBiasClamp {bInterpResult=true, interpSpeedDecreasing=2}
    boneToModify {boneName=dyn_hat_04}
    translation {x=2}
   }
   n32 = "Transform (Modify) Bone - Bone: dyn_hat_03"(
    ComponentPose=@n31
    Alpha=Get Hat_transform_Level_Targeting() -> (Hat_transform_Level_Targeting:double)
    Translation="10,0,0"
    Rotation="0,0,0"
    Scale="1,1,1"
   ) {
    rotationMode=BMM_Additive
    rotationSpace=BCS_ParentBoneSpace
    translationMode=BMM_Additive
    translationSpace=BCS_ParentBoneSpace
    alphaBoolBlend {blendOption=Linear}
    alphaScaleBiasClamp {interpSpeedDecreasing=8, interpSpeedIncreasing=2}
    boneToModify {boneName=dyn_hat_03}
    translation {x=10}
   }
   n33 = "Constraint - hat_02 (PPPP)"!(
    ComponentPose=@n32
    bAlphaBoolEnabled:bool=True
    ConstraintWeights_0=0.136
    ConstraintWeights_1:float=0.127
    ConstraintWeights_2:float=0.151
    ConstraintWeights_3:float=0.586
   ) -> (Pose) {
    alphaInputType=Bool
    alphaBoolBlend {blendOption=Linear}
    boneToModify {boneName=hat_02}
    constraintSetup [{targetBone {boneName=Hat}}, {targetBone {boneName=dyn_hat_03}}, {targetBone {boneName=dyn_hat_04}}, {targetBone {boneName=hat_02}}]
    constraintWeights [1,1,1,1]
   }
   n34 = "Transform (Modify) Bone - Bone: hat_02"!(ComponentPose=@n33, bAlphaBoolEnabled:bool=True, Translation="0.6,0,0", Rotation="0,0,0", Scale="1,1,1") -> (Pose) {
    alphaInputType=Bool
    rotationMode=BMM_Additive
    rotationSpace=BCS_ParentBoneSpace
    translationMode=BMM_Additive
    translationSpace=BCS_ParentBoneSpace
    alphaBoolBlend {blendOption=Linear}
    alphaScaleBiasClamp {bInterpResult=true, interpSpeedIncreasing=5}
    boneToModify {boneName=hat_02}
    translation {x=0.6}
   }
   n35 = "Transform (Modify) Bone - Bone: Hat"(ComponentPose=@n34, bAlphaBoolEnabled=True, Translation="0.2,0,0", Rotation="0,0,0", Scale="1,1,1") {
    alphaInputType=Bool
    rotationMode=BMM_Additive
    rotationSpace=BCS_ParentBoneSpace
    translationMode=BMM_Additive
    translationSpace=BCS_BoneSpace
    alphaBoolBlend {blendOption=Linear}
    alphaScaleBiasClamp {bInterpResult=true, interpSpeedIncreasing=5}
    boneToModify {boneName=Hat}
    translation {x=0.2}
   }
   n36 = "Transform (Modify) Bone - Bone: dyn_rope_01"(ComponentPose=@n35, Alpha=1, Translation="0,0,0", Rotation="5,0,0", Scale="1,1,1") {
    rotationMode=BMM_Additive
    rotationSpace=BCS_ParentBoneSpace
    translationMode=BMM_Additive
    translationSpace=BCS_ParentBoneSpace
    alphaBoolBlend {blendOption=Linear}
    boneToModify {boneName=dyn_rope_01}
    rotation {pitch=5}
   }
   n37 = Trail controller - Bone: dyn_rope_08(ComponentPose=@n36, Alpha=1) {
    bActorSpaceFakeVel=true
    bLimitRotation=true
    bUsePlanarLimit=true
    chainLength=7
    baseJoint {boneName=root}
    fakeVelocity {z=30}
    planarLimits [{drivingBone {boneName=spine_05}, planeTransform {location=<0 13 0>, rotation=<rot 0 -15 90>, scale=<1.5 1 1>}}]
    rotationLimits [
     {limitMax=<30 30 50>, limitMin=<-30 -30 0>}
     {limitMax=<30 30 50>, limitMin=<-30 -30 0>}
     {limitMax=<30 30 50>, limitMin=<-30 -30 10>}
     {limitMax=<30 30 30>, limitMin=<-30 -30 10>}
     {limitMax=<30 30 30>, limitMin=<-30 -30 10>}
     {limitMax=<30 30 30>, limitMin=<-30 -30 10>}
     {limitMax=<30 30 30>, limitMin=<-30 -30 10>}
    ]
    rotationOffsets [<0 0 0>, <0 0 0>, <0 0 0>, <0 0 0>, <0 0 0>, <0 0 0>, <0 0 0>]
    trailBone {boneName=dyn_rope_08}
    trailRelaxationSpeed {editorCurveData {keys [<at 0.2757 8.4671 RCIM_Linear>,<at 0.8709 5.8144 RCIM_Linear>]}}
   }
   Output Pose(Result=Blend Poses by bool(BlendPose_0=Component To Local(ComponentPose=@n16), BlendPose_1=Component To Local(ComponentPose=@n37), bActiveValue=Get bRigidBodyEnabled()) {blendType=Linear}) {layerGroup=None}
   Reroute Node(InputPin:float=Reroute Node()) -> (OutputPin:float)
   Get AlphaHeadRapidRotation() -> (AlphaHeadRapidRotation:float)
  ]}
 EventGraph {nodes [
   n1 = Get SkeletalMesh() -> (SkeletalMesh:Skeletal Mesh Component Object Reference)
   n2 = Calculate Velocity From Position History!(
    execute
    DeltaSeconds:float=0
    Position:Vector=MakeVector(
     Y=0
     Z=GetAlphaFromReferenceTransformDelta(SkeletalMeshComponent=@n1, BoneName=head, BaseName=spine_05, InputRange=(X=0,Y=-90), TwistAxis="1,0,0").RawOutput
    )
    History:"Position History Structure (by ref)"=Get HistoryHeadYaw() -> (HistoryHeadYaw:Position History Structure)
    NumberOfSamples:int=8
    VelocityMin:float=20
    VelocityMax:float=100
   ) -> (then, ReturnValue:float)
   > Set AlphaHeadRapidRotation(AlphaHeadRapidRotation:float=1) -> (Output_Get:float)
   Event BlueprintUpdateAnimation()
   > Set head_shape_bk(head_shape_bk:double=GetAlphaFromReferenceTransformDelta(
     SkeletalMeshComponent=Get FaceAcc_Skel() -> (FaceAcc_Skel:Skeletal Mesh Component Object Reference)
     BoneName=head
     BaseName=spine_05
     InputRange=(X=0,Y=-170)
     TwistAxis="0,0,1"
    )) -> (Output_Get:double)
   n3 = Vector Length(A=Get PawnVelocity())
   n4 = "To Float (Boolean)"(InBool=OR Boolean(A=Get bIsCrouchSprinting(), B=Get bIsSprinting()))
   > Set RigidBodyAlpha(RigidBodyAlpha:double=MapRangeClamped(Value=float * float(A=@n3, B=@n4) {is=title K2Node_PromotableOperator}, InRangeA=100, InRangeB=770, OutRangeA=1, OutRangeB=0.5)) -> (Output_Get:double)
   n5 = "To Float (Boolean)"(InBool=OR Boolean(A=Get bIsCrouchSprinting(), B=Get bIsTacticalSprinting()))
   > Set Hat_transform(Hat_transform:double=MapRangeClamped(Value=float * float(A=@n3, B=@n5), InRangeA=20, InRangeB=770)) -> (Output_Get:double)
   n6 = Get SkeletalMesh()
   > Set Crouch_Hat's_transform(Crouch_Hat's_transform:double=GetAlphaFromReferenceTransformDelta(SkeletalMeshComponent=@n6, BoneName=head, BaseName=spine_05, InputRange=(X=-25,Y=-65), TwistAxis="1,0,0")) -> (Output_Get:double)
   n7 = Get bIsTargeting()
   > Set TargetingCorrection(TargetingCorrection:float=float * float(
     A=GetAlphaFromReferenceTransformDelta(SkeletalMeshComponent=@n6, BoneName=head, BaseName=spine_05, InputRange=(X=-45,Y=-80), TwistAxis="1,0,0")
     B="To Float (Boolean)"(InBool=@n7)
    )) -> (Output_Get:float)
   n8 = Get FaceAcc_Skel()
   n9 = Get bIsPlayingEmote()
   > Set Emote_Hat_transform(Emote_Hat_transform:double=float * float(
     A=GetAlphaFromReferenceTransformDelta(SkeletalMeshComponent=@n8, BoneName=hat, BaseName=spine_05, InputRange=(X=-30,Y=45), TwistAxis="1,0,0")
     B="To Float (Boolean)"(InBool=@n9)
    )) -> (Output_Get:double)
   n10 = Get bIsPlayingEmote()
   > Set Emote_Hat_transform_2(Emote_Hat_transform_2:double=float * float(
     A=GetAlphaFromReferenceTransformDelta(SkeletalMeshComponent=@n6, BoneName=clavicle_r, BaseName=spine_05, InputRange=(X=0,Y=15), TwistAxis="1,0,0")
     B="To Float (Boolean)"(InBool=@n10)
    )) -> (Output_Get:double)
   n11 = NOT Boolean(A=Get bIsPlayingEmote())
   n12 = Get bIsInFrontEnd()
   > Set Hat_transform_Frontend(Hat_transform_Frontend:double=SelectFloat(A=1, B=0, bPickA=AND Boolean(A=@n12, B=@n11))) -> (Output_Get:double)
   n13 = Get bIsInFrontEnd()
   n14 = Get SkeletalMesh()
   > Set Hat_transform_noRB_emote(Hat_transform_noRB_emote:double=float * float(
     A=GetAlphaFromReferenceTransformDelta(SkeletalMeshComponent=@n14, BoneName=head, BaseName=spine_05, InputRange=(X=5,Y=-45), TwistAxis="1,0,0")
     B="To Float (Boolean)"(InBool=@n13)
    )) -> (Output_Get:double)
   > Set Hat_transform_Level(Hat_transform_Level:double=GetAlphaFromReferenceTransformDelta(SkeletalMeshComponent=@n14, BoneName=head, BaseName=spine_05, InputRange=(X=5,Y=-42), TwistAxis="1,0,0")) -> (Output_Get:double)
   n15 = GetAlphaFromReferenceTransformDelta(SkeletalMeshComponent=@n14, BoneName=head, BaseName=spine_05, InputRange=(X=-35,Y=-85), TwistAxis="1,0,0")
   n16 = GetAlphaFromReferenceTransformDelta(SkeletalMeshComponent=@n14, BoneName=clavicle_r, BaseName=spine_05, DeltaComponentType=EulerY, InputRange=(X=5,Y=14), TwistAxis="1,0,0")
   n17 = float * float(
    A=GetAlphaFromReferenceTransformDelta(SkeletalMeshComponent=@n14, BoneName=head, BaseName=spine_05, InputRange=(X=-10,Y=-42), TwistAxis="1,0,0")
    B=@n16
   )
   > Set Hat_transform_Level_headUp(Hat_transform_Level_headUp:double="Clamp (Float)"(Value=float + float(A=@n15, B=@n17), Min=0, Max=1)) -> (Output_Get:double)
   > Set Hat_transform_Level_Targeting(Hat_transform_Level_Targeting:double=float * float(
     A=GetAlphaFromReferenceTransformDelta(SkeletalMeshComponent=@n14, BoneName=head, BaseName=spine_05, DeltaComponentType=EulerY, InputRange=(X=25,Y=38), TwistAxis="1,0,0")
     B=@n16
    )) -> (Output_Get:double)
   Event OnCharacterCustomizationLoaded()
   > Set SkeletalMesh(SkeletalMesh:Skeletal Mesh Component Object Reference=GetOwnerSkeletalMeshComponent()) -> (Output_Get:Skeletal Mesh Component Object Reference)
   > n18 = Cast To PlayerPawn_Athena(Object=TryGetPawnOwner())
   > Set FaceAcc_Skel(FaceAcc_Skel:Skeletal Mesh Component Object Reference=GetSkeletalMeshForPartType(self=@n18, PartType=Face)) -> (Output_Get:Skeletal Mesh Component Object Reference)
   float * float(A=@n2, B="To Float (Boolean)"(InBool=Get bIsTargeting()))
  ]}
]
variables [
 AlphaHeadRapidRotation {type=number}
 Crouch_Hat's_transform {type=number}
 Emote_Hat_transform {type=number}
 Emote_Hat_transform_2 {type=number}
 FaceAcc {type=CustomCharacterPartModifier, default=null}
 FaceAcc_Skel {type=SkeletalMeshComponent}
 Hat_transform {type=number}
 Hat_transform_Frontend {type=number}
 Hat_transform_Level {type=number}
 Hat_transform_Level_headUp {type=number}
 Hat_transform_Level_Targeting {type=number}
 Hat_transform_noRB_emote {type=number}
 head_shape_bk {type=number}
 HistoryHeadYaw {type=PositionHistory, default {range=0, positions []}}
 RigidBodyAlpha {type=number}
 SkeletalMesh {type=SkeletalMeshComponent}
 TargetingCorrection {type=number}
]
